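# Animation Blueprint for the Astronaut 01 male medium player body, which follows the pose of another skeletal mesh.
# On initialization it caches the owner's skeletal mesh, then copies that pose and constrains neck_01 between head and spine_05, weighted toward the head.
# A separate chain of Bone Driven Controllers maps head rotation to the helmet by HelmetAlpha, but it is not connected to the output pose.
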AnimBlueprint /BRCosmetics/Characters/Player/Male/Medium/Bodies/M_MED_Astronaut_01/Meshes/M_MED_Astronaut_01_Skeleton_AnimBP
target_skeleton=/Game/Characters/Player/Male/Medium/Bodies/M_MED_Astronaut_01/Meshes/M_MED_Astronaut_01_Skeleton
graphs [
 AnimGraph {nodes [
   n1 = Get HelmetAlpha() -> (HelmetAlpha:double)
   n2 = "helmet.rotateY = head.rotateY - Bone Driven Controller"!(ComponentPose:Component Space Pose Link Structure=Local To Component(), Alpha:float=@n1) -> (Pose:Component Space Pose Link Structure) {
    is=title AnimGraphNode_BoneDrivenController
    bAffectTargetRotationY=true
    sourceComponent=RotationY
    alphaBoolBlend {blendOption=Linear}
    sourceBone {boneName=head}
    targetBone {boneName=helmet}
   }
   n3 = "helmet.rotateY = head.rotateY - Bone Driven Controller"(ComponentPose=@n2, Alpha=@n1) {
    bAffectTargetRotationY=true
    sourceComponent=RotationY
    alphaBoolBlend {blendOption=Linear}
    sourceBone {boneName=head}
    targetBone {boneName=helmet}
   }
   n4 = "helmet.rotateZ = head.rotateZ - Bone Driven Controller"!(ComponentPose:Component Space Pose Link Structure=@n3, Alpha:float=@n1) -> (Pose:Component Space Pose Link Structure) {
    is=title AnimGraphNode_BoneDrivenController
    bAffectTargetRotationZ=true
    sourceComponent=RotationZ
    alphaBoolBlend {blendOption=Linear}
    sourceBone {boneName=head}
    targetBone {boneName=helmet}
   }
   Component To Local(ComponentPose=@n4)
   n5 = Get MeshToCopy() -> (MeshToCopy:Skeletal Mesh Component Object Reference)
   n6 = "Constraint - neck_01 (PP)"!(
    ComponentPose=Local To Component(LocalPose=Copy Pose From Mesh(SourceMeshComponent=@n5))
    Alpha=1
    ConstraintWeights_0=0.6
    ConstraintWeights_1:float=0.4
   ) -> (Pose) {
    alphaBoolBlend {blendOption=Linear}
    alphaScaleBiasClamp {bInterpResult=true, interpSpeedDecreasing=5, interpSpeedIncreasing=7}
    boneToModify {boneName=neck_01}
    constraintSetup [{targetBone {boneName=head}}, {targetBone {boneName=spine_05}}]
    constraintWeights [1,1]
   }
   Output Pose(Result=Component To Local(ComponentPose=@n6)) {layerGroup=None}
  ]}
 EventGraph {nodes [
   Event BlueprintInitializeAnimation()
   > Set MeshToCopy(MeshToCopy:Skeletal Mesh Component Object Reference=GetOwnerSkeletalMeshComponent()) -> (Output_Get:Skeletal Mesh Component Object Reference)
   TryGetPawnOwner()
   Event BlueprintUpdateAnimation()
  ]}
]
variables [
 HelmetAlpha {type=number, default=0.5}
 MeshToCopy {type=SkeletalMeshComponent}
]
